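# Editor-only callback Blueprint that notifies a Python `nres_bridge` module of an event, by default that recording has finished.
# On BeginPlay, it checks the Execute flag before running a Python command built by string concatenation.
# The command imports the bridge and calls `send_event` with SendString and Delay, so any wait is left to Python rather than handled in the graph.
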EditorUtilityBlueprint /Game/NeverCook/NRES/BP_FADE_CallBack
graphs [
 EventGraph {nodes [
   Event BeginPlay()
   > Branch(Condition=Get Execute() -> (Execute:bool))
   n1 = Get Delay() -> (Delay:double)
   > ExecutePythonCommand(PythonCommand=Append!(
     A="import nres_bridge\r\nnres_bridge.send_event('"
     B=Get SendString() -> (SendString:String)
     C:String="', delay="
     D:String="To String (Float)"(InDouble=@n1)
     E:String=") "
    ) -> (ReturnValue))
  ]}
 UserConstructionScript {nodes [Construction Script()]}
]
variables [
 Delay {type=number, default=2}
 Execute {type=boolean}
 SendString {type=string, default=FinishedRecording}
]
