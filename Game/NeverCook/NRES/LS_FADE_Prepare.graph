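# LS_FADE_Prepare is a Level Sequence that, by its binding names, runs a short FADE preparation pass for Movie Render Queue.
# A camera cut track spans the sequence, with a CineCamera bound without animation.
# The MRQNS Niagara actor is deactivated, briefly activated, then deactivated again.
# Just afterwards, a spawn track brings in the FinishMRQ BP_FADE_CallBack actor, apparently as a completion callback.
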LevelSequence /Game/NeverCook/NRES/LS_FADE_Prepare
bindings [
 CineCameraActor {id=#1, possessed_class=CineCameraActor, tracks []}
 FinishMRQ {
  id=#2
  possessed_class=BP_FADE_CallBack_C
  tracks [
   {
    type=MovieSceneSpawnTrack
    sections [{type=MovieSceneSpawnSection, channels [None {type=MovieSceneScriptingBoolChannel, default=true, keys [[0,false],[23,true]]}]}]
   }
   {
    type=MovieScene3DTransformTrack
    sections [{
      type=MovieScene3DTransformSection
      channels [
       Location.X {type=MovieSceneScriptingDoubleChannel, default=0}
       Location.Y {type=MovieSceneScriptingDoubleChannel, default=0}
       Location.Z {type=MovieSceneScriptingDoubleChannel, default=0}
       Rotation.X {type=MovieSceneScriptingDoubleChannel, default=0}
       Rotation.Y {type=MovieSceneScriptingDoubleChannel, default=0}
       Rotation.Z {type=MovieSceneScriptingDoubleChannel, default=0}
       Scale.X {type=MovieSceneScriptingDoubleChannel, default=1}
       Scale.Y {type=MovieSceneScriptingDoubleChannel, default=1}
       Scale.Z {type=MovieSceneScriptingDoubleChannel, default=1}
      ]
     }]
   }
  ]
 }
 MRQNS {id=#3, possessed_class=NiagaraActor, tracks []}
 CameraComponent {id=#4, parent_id=#1, possessed_class=CineCameraComponent, tracks []}
 NiagaraComponent0 {
  id=#5
  parent_id=#3
  possessed_class=NiagaraComponent
  tracks [{
    type=MovieSceneParticleTrack
    display_name=FX System
    sections [{
      type=MovieSceneParticleSection
      channels [None {
        type=MovieSceneScriptingParticleChannel
        keys [[1, "<ParticleKey.DEACTIVATE: 1>"], [2.235, "<ParticleKey.ACTIVATE: 0>"], [22, "<ParticleKey.DEACTIVATE: 1>"]]
       }]
     }]
   }]
 }
]
display_rate {denominator=1, numerator=30}
tick_resolution {denominator=1, numerator=24000}
tracks [{type=MovieSceneCameraCutTrack, sections [{type=MovieSceneCameraCutSection, end_frame=150, start_frame=0}]}]
