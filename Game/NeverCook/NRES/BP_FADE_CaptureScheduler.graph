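# Editor utility blueprint that schedules Niagara effect capture steps: Prepare, StartRecord and BoundCheck.
# Each event assigns the NS system to the first NiagaraActor found, clearing parameter overrides, and reactivates it with a reset.
# It then sets a half-second, one-shot timer by name to call the same-named function on BP_FXCaptureHelper.
# Camera and level sequence variables are declared but unused in these graphs.
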EditorUtilityBlueprint /Game/NeverCook/NRES/BP_FADE_CaptureScheduler
graphs [
 EventGraph {nodes [
   Prepare() {is=title K2Node_CustomEvent}
   > n1 = GetAllActorsOfClass!(execute, WorldContextObject:Object Reference, ActorClass:Actor Class Reference=/Script/Niagara.NiagaraActor) -> (then, OutActors:Array of Niagara Particle System Actor Object References)
   n2 = Get NiagaraComponent!(self:Niagara Particle System Actor Object Reference="Get (a copy)"(Array:Array of Niagara Particle System Actor Object References=@n1) -> (Output:Niagara Particle System Actor Object Reference)) -> (NiagaraComponent:Niagara Particle System Component Object Reference)
   > Set Niagara System Asset!(
    execute
    self:Niagara Particle System Component Object Reference=@n2
    InAsset:Niagara System Object Reference=Get NS() -> (NS:Niagara System Object Reference)
    bResetExistingOverrideParameters:bool=true
   ) -> (then) {is=title K2Node_CallFunction}
   > SetActive!(execute, self:Actor Component Object Reference=@n2, bNewActive:bool=true, bReset:bool=true) -> (then)
   > Set Timer by Function Name!(
    execute
    Object:Object Reference=Get BP_FXCaptureHelper() -> (BP_FXCaptureHelper:BP FADE FXCapture Helper Object Reference)
    FunctionName:String=Prepare
    Time:float=0.5
    bLooping:bool=false
    bMaxOncePerFrame:bool=false
    InitialStartDelay:float=0
    InitialStartDelayVariance:float=0
   ) -> (then, ReturnValue:Timer Handle Structure) {is=title K2Node_CallFunction}
   StartRecord() {is=title K2Node_CustomEvent}
   > n3 = GetAllActorsOfClass(ActorClass=&NiagaraActor)
   n4 = Get NiagaraComponent(self="Get (a copy)"(Array=@n3))
   > Set Niagara System Asset(self=@n4, InAsset=Get NS(), bResetExistingOverrideParameters=true)
   > SetActive(self=@n4, bNewActive=true, bReset=true)
   > Set Timer by Function Name(
    Object=Get BP_FXCaptureHelper()
    FunctionName=StartRecord
    Time=0.5
    bLooping=false
    bMaxOncePerFrame=false
    InitialStartDelay=0
    InitialStartDelayVariance=0
   )
   BoundCheck() {is=title K2Node_CustomEvent}
   > n5 = GetAllActorsOfClass(ActorClass=&NiagaraActor)
   n6 = Get NiagaraComponent(self="Get (a copy)"(Array=@n5))
   > Set Niagara System Asset(self=@n6, InAsset=Get NS(), bResetExistingOverrideParameters=true)
   > SetActive(self=@n6, bNewActive=true, bReset=true)
   > Set Timer by Function Name(
    Object=Get BP_FXCaptureHelper()
    FunctionName=BoundCheck
    Time=0.5
    bLooping=false
    bMaxOncePerFrame=false
    InitialStartDelay=0
    InitialStartDelayVariance=0
   )
  ]}
 UserConstructionScript {nodes [Construction Script()]}
]
variables [
 CameraPos {type=Vector}
 CameraRotation {type=Rotator}
 LS {type=LevelSequence, default=null}
 LS_Prepare {type=LevelSequence, default=&/LS_FADE_Prepare}
 NS {type=NiagaraSystem, default=null, class=~, dependencies=~, emitters=~, summary=~}
]
