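; BP_FADE_FXCaptureHelper is an editor-side helper, owned by BP_FADE_CaptureScheduler, that stages FX capture passes.
; BoundCheck, Prepare and StartRecord position a CineCamera: fixed for the first two, scheduler-supplied for recording.
; Each edits Sequencer binding templates to set BP_FADE_CallBack completion strings; BoundCheck and StartRecord also toggle BP_FADE_NSBoundChecker's SendBound.
; Those two then render the MPQ_FADE_Major queue through Movie Render Pipeline's PIE executor.
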(event Custom|BoundCheck
  (bind _returnvalue (EditorSubsystems|GetMoviePipelineQueueSubsystem))
  (bind _asbp_fade_capture_scheduler (Utilities|Casting|CastToBP_FADE_CaptureScheduler (Components|GetOwner)))
  (bind _returnvalue_2 (Sequencer|Sequence|GetBindings (Variables|/Game/NeverCook/NRES/BP_FADE_CaptureScheduler.BP_FADE_CaptureScheduler_C|GetLS _asbp_fade_capture_scheduler)))
  (Utilities|FlowControl|Sequence
    (:then_0
      (bind _outactors (Actor|GetAllActorsOfClass 0 "/Script/CinematicCamera.CineCameraActor"))
      (Transformation|SetActorLocation (Utilities|Array|Get(acopy) _outactors) "369.000000,-369.000000,369.000000" false true)
      (Transformation|SetActorRotation (Utilities|Array|Get(acopy) _outactors) "-35.260000,135.000000,0.000000" true))
    (:then_2
      (for _array_element _returnvalue_2
        (bind _returnvalue_3 (Sequencer|Sequence|GetObjectTemplate _array_element))
        (Utilities|FlowControl|Sequence
          (:then_0
            (bind _asbp_fade_nsbound_checker (Utilities|Casting|CastToBP_FADE_NSBoundChecker _returnvalue_3))
            (Variables|/Game/NeverCook/NRES/BP_FADE_NSBoundChecker.BP_FADE_NSBoundChecker_C|SetSendBound true _asbp_fade_nsbound_checker))
          (:then_1
            (bind _asbp_fade_call_back (Utilities|Casting|CastToBP_FADE_CallBack _returnvalue_3))
            (Variables|/Game/NeverCook/NRES/BP_FADE_CallBack.BP_FADE_CallBack_C|SetExecute true _asbp_fade_call_back)
            (Variables|/Game/NeverCook/NRES/BP_FADE_CallBack.BP_FADE_CallBack_C|SetSendString "BoundFinished" _asbp_fade_call_back))))
      (if (MovieRenderPipeline|LoadQueue _returnvalue "/Game/NeverCook/NRES/MPQ_FADE_Major.MPQ_FADE_Major" false)
        (MovieRenderPipeline|Rendering|RenderQueuewithExecutor _returnvalue "/Script/MovieRenderPipelineEditor.MoviePipelinePIEExecutor" :!class /Script/MovieRenderPipelineEditor.MoviePipelineQueueSubsystem)))))

(event Custom|Prepare
  (bind _asbp_fade_capture_scheduler (Utilities|Casting|CastToBP_FADE_CaptureScheduler (Components|GetOwner)))
  (Utilities|FlowControl|Sequence
    (:then_0
      (bind _outactors (Actor|GetAllActorsOfClass 0 "/Script/CinematicCamera.CineCameraActor"))
      (Transformation|SetActorLocation (Utilities|Array|Get(acopy) _outactors) "369.000000,-369.000000,369.000000" false true)
      (Transformation|SetActorRotation (Utilities|Array|Get(acopy) _outactors) "-35.260000,135.000000,0.000000" true))
    (:then_2
      (bind _returnvalue (Sequencer|Sequence|GetBindings (Variables|/Game/NeverCook/NRES/BP_FADE_CaptureScheduler.BP_FADE_CaptureScheduler_C|GetLS_Prepare _asbp_fade_capture_scheduler)))
      (for _array_element _returnvalue
        (bind _returnvalue_2 (Sequencer|Sequence|GetObjectTemplate _array_element))
        (Utilities|FlowControl|Sequence
          (:then_1
            (bind _asbp_fade_call_back (Utilities|Casting|CastToBP_FADE_CallBack _returnvalue_2))
            (Variables|/Game/NeverCook/NRES/BP_FADE_CallBack.BP_FADE_CallBack_C|SetExecute true _asbp_fade_call_back)
            (Variables|/Game/NeverCook/NRES/BP_FADE_CallBack.BP_FADE_CallBack_C|SetSendString "PrepareFinished" _asbp_fade_call_back)))))))

(event Custom|StartRecord
  (bind _returnvalue (EditorSubsystems|GetMoviePipelineQueueSubsystem))
  (bind _asbp_fade_capture_scheduler (Utilities|Casting|CastToBP_FADE_CaptureScheduler (Components|GetOwner)))
  (bind _returnvalue_2 (Sequencer|Sequence|GetBindings (Variables|/Game/NeverCook/NRES/BP_FADE_CaptureScheduler.BP_FADE_CaptureScheduler_C|GetLS _asbp_fade_capture_scheduler)))
  (Utilities|FlowControl|Sequence
    (:then_1
      (bind _outactors (Actor|GetAllActorsOfClass 0 "/Script/CinematicCamera.CineCameraActor"))
      (Transformation|SetActorLocation (Utilities|Array|Get(acopy) _outactors) (Variables|/Game/NeverCook/NRES/BP_FADE_CaptureScheduler.BP_FADE_CaptureScheduler_C|GetCameraPos _asbp_fade_capture_scheduler) false true)
      (Transformation|SetActorRotation (Utilities|Array|Get(acopy) _outactors) (Variables|/Game/NeverCook/NRES/BP_FADE_CaptureScheduler.BP_FADE_CaptureScheduler_C|GetCameraRotation _asbp_fade_capture_scheduler) true))
    (:then_2
      (for _array_element _returnvalue_2
        (bind _returnvalue_3 (Sequencer|Sequence|GetObjectTemplate _array_element))
        (Utilities|FlowControl|Sequence
          (:then_0
            (bind _asbp_fade_nsbound_checker (Utilities|Casting|CastToBP_FADE_NSBoundChecker _returnvalue_3))
            (Variables|/Game/NeverCook/NRES/BP_FADE_NSBoundChecker.BP_FADE_NSBoundChecker_C|SetSendBound false _asbp_fade_nsbound_checker))
          (:then_1
            (bind _asbp_fade_call_back (Utilities|Casting|CastToBP_FADE_CallBack _returnvalue_3))
            (Variables|/Game/NeverCook/NRES/BP_FADE_CallBack.BP_FADE_CallBack_C|SetExecute true _asbp_fade_call_back)
            (Variables|/Game/NeverCook/NRES/BP_FADE_CallBack.BP_FADE_CallBack_C|SetSendString "FinishedRecording" _asbp_fade_call_back))))
      (if (MovieRenderPipeline|LoadQueue _returnvalue "/Game/NeverCook/NRES/MPQ_FADE_Major.MPQ_FADE_Major" false)
        (MovieRenderPipeline|Rendering|RenderQueuewithExecutor _returnvalue "/Script/MovieRenderPipelineEditor.MoviePipelinePIEExecutor" :!class /Script/MovieRenderPipelineEditor.MoviePipelineQueueSubsystem)))))
