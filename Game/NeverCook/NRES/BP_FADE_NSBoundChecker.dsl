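; Blueprint that reports a Niagara effect's local bounds to the external Python module nres_bridge.
; Each tick with SendBound on, it sends the Target component's origin, box extent and sphere radius as a "BoundData" event through ExecutePythonCommand, then prints the extent.
; Without a valid Target, it scans Niagara actors for a name containing "MRQNS", but assigns the array's default-index (first) actor rather than the matched one.
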(event EventTick
  (params (DeltaSeconds :type Float%20%28single-precision%29))
  (bind _target (Variables|Default|GetTarget))
  (bind _niagaracomponent (Variables|/Script/Niagara.NiagaraActor|GetNiagaraComponent _target))
  (bind _returnvalue (Fortnite|Utility|GetLocalBounds _niagaracomponent))
  (bind (_origin _boxextent _sphereradius) (Math|BoxSphereBounds|BreakBoxSphereBounds _returnvalue))
  (if (Variables|Default|GetSendBound)
    (Utilities|IsValid (Variables|Default|GetTarget)
      (:Is%20Valid
        (Python|Execution|ExecutePythonCommand (Utilities|String|Append "import nres_bridge
nres_bridge.send_event('BoundData',
    value='" (Utilities|String|ToString(Vector) _origin) "&&" (Utilities|String|ToString(Vector) _boxextent) "&&" (Utilities|String|ToString(Float) _sphereradius) "') "))
        (Development|PrintString (Utilities|String|ToString(Vector) _boxextent)))
      (:Is%20Not%20Valid
        (bind _outactors (Actor|GetAllActorsOfClass "/Script/Niagara.NiagaraActor"))
        (bind _array_element (Utilities|Array|ForEachLoopwithBreak _outactors)
          (:LoopBody
            (if (>= (Utilities|String|FindSubstring (Utilities|GetDisplayName _array_element) "MRQNS") 0)
              (Variables|Default|SetTarget (Utilities|Array|Get(acopy) _outactors)))))))))

(fn ConstructionScript)
